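# M_PolarUV is a Material Gym example material that shows polar UV mapping on the UV editor color grid texture.
# The MF_Polar function remaps texture coordinates using length and radial scale parameters, leaving its center offset and twist inputs unconnected.
# The texture sample uses explicit DDX/DDY derivatives of the unwarped UVs for mip selection, which avoids mip seams where the polar coordinates wrap.
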Material /Game/Global/Materials/MaterialGym/Materials/MaterialElements/UVs/UVMainpulation/PolarUV/M_PolarUV
graph {
 properties {bAutomaticallySetUsageInEditor=false, bHasPixelAnimation=true, usedShadingModels=Default Lit}
 nodes [
  n1 = TextureCoordinate()
  TextureSample!(
   UVs=MaterialFunctionCall!(
    UV_Input=TextureCoordinate()
    PolarUV_CenterOffsetG
    PolarUV_CenterOffsetR
    PolarUV_LenghtScale=ScalarParameter() {defaultValue=0.714, parameterName=Polar_LenghtScale}
    PolarUV_RadialScale=ScalarParameter() {defaultValue=2, parameterName=Polar_RadialScale}
    PolarUV_UseTwist
    PolarUV_TwistOffset
    PolarUVTwistSpeed
    PolarUV_TwistRadius
   ) {materialFunction=&/MF_Polar}.Result
   Tex
   "DDX(UVs)"=DDX(Value=@n1)
   "DDY(UVs)"=DDY(Value=@n1)
   Apply View MipBias
  ) {mipValueMode=TMVM_Derivative, samplerSource=SSM_Wrap_WorldGroupSettings, texture=/UVEditor/Textures/UVEditorColorGrid_Color}
 ]
}
