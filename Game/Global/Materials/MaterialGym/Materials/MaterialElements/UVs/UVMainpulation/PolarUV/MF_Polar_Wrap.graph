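# Material function that wraps MF_Polar, exposing its polar UV remap as ready-made, named material parameters in the MF_PolarUV group.
# UV and center offset are overridable function inputs defaulting to texture coordinates and scalar parameters.
# Length and radial scale, a static twist switch, and twist offset, speed and radius feed straight in as parameters.
# It outputs the resulting UV plus a twist mask for debugging.
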MaterialFunction /Game/Global/Materials/MaterialGym/Materials/MaterialElements/UVs/UVMainpulation/PolarUV/MF_Polar_Wrap
graph {nodes [
  n1 = MaterialFunctionCall!(
   UV_Input=FunctionInput(Preview=TextureCoordinate()) {bUsePreviewValueAsDefault=true, inputName=UVInput, inputType=FunctionInput_Vector2}
   PolarUV_CenterOffsetG=FunctionInput(Preview=ScalarParameter() {defaultValue=-0.5, group=MF_PolarUV, parameterName=PolarUV_CenterOffsetG}) {bUsePreviewValueAsDefault=true, inputName=PolarUV_CenterOffsetG, inputType=FunctionInput_Scalar, sortPriority=1}
   PolarUV_CenterOffsetR=FunctionInput(Preview=ScalarParameter() {defaultValue=-0.5, group=MF_PolarUV, parameterName=PolarUV_CenterOffsetR}) {bUsePreviewValueAsDefault=true, inputName=PolarUV_CenterOffsetR, inputType=FunctionInput_Scalar, sortPriority=1}
   PolarUV_LenghtScale=ScalarParameter() {defaultValue=1, group=MF_PolarUV, parameterName=PolarUV_LenghtScale}
   PolarUV_RadialScale=ScalarParameter() {defaultValue=1, group=MF_PolarUV, parameterName=PolarUV_RadialScale}
   PolarUV_UseTwist=StaticBoolParameter() {defaultValue=true, group=MF_PolarUV, parameterName=PolarUV_UseTwist}
   PolarUV_TwistOffset=AppendVector(
    A=ScalarParameter() {group=MF_PolarUV, parameterName=PolarUV_TwistTilingU}
    B=ScalarParameter() {defaultValue=-0.55, group=MF_PolarUV, parameterName=PolarUV_TwistTilingV}
   )
   PolarUVTwistSpeed=ScalarParameter() {defaultValue=0.05, group=MF_PolarUV, parameterName=PolarUV_TwistSpeed}
   PolarUV_TwistRadius=ScalarParameter() {defaultValue=0.3499, group=MF_PolarUV, parameterName=PolarUV_TwistRadius}
  ) {materialFunction=&/MF_Polar}
  FunctionOutput(None=@n1.Result) {outputName=OutputUV}
  FunctionOutput(None=@n1.DebugTwistMask) {outputName=DebugTwistMask, sortPriority=1}
 ]}
